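# Material function for the QuarterClasp Zoom Cap gadget prop that tints the base color of incoming material attributes within a masked area.
# It multiplies base color by a boosted RandomColorFromAnimNotify vector parameter, whose name suggests an anim notify drives it.
# The blue channel of a mask, apparently the TechArt mask input, blends between original and tinted color.
# Only BaseColor is overridden on the output attributes.
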MaterialFunction /BRCosmetics/Animation/Props/Gadgets_QuarterClaspZoom_Cap/Materials/MF_QuarterClasp_Zoom_Cap_TechArt
graph {nodes [
  n1 = GetMaterialAttributes(None=NamedRerouteDeclaration(None=FunctionInput() {inputName=In_MaterialAttributes, inputType=FunctionInput_MaterialAttributes}) {name=In_MaterialAttributes, nodeColor=<rgba 1 0 0.8437 1>}) {attributeGetTypes [#1]}
  n2 = VectorParameter() {group=QuarterClaspZoom_Cap_TechArt, parameterName=RandomColorFromAnimNotify, defaultValue=<rgba 1 0 0.05 1>}
  n3 = Multiply(A=@n1.BaseColor, B=Multiply(A=@n2.RGB) {constB=3})
  n4 = ComponentMask(None=NamedRerouteUsage()) {b=true}
  FunctionOutput(None=SetMaterialAttributes!(MaterialAttributes=@n1.MaterialAttributes, BaseColor=LinearInterpolate(A=@n1.BaseColor, B=@n3, Alpha=@n4)) {attributeSetTypes [#1]})
  NamedRerouteDeclaration(None=FunctionInput() {inputName=In_TechArtMask, sortPriority=1}) {name=In_TechArtMask, nodeColor=<rgba 0.5234 1 0 1>}
 ]}
