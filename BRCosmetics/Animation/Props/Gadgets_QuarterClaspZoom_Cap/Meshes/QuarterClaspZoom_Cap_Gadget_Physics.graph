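# Physics asset for the QuarterClaspZoom Cap gadget prop in BRCosmetics, giving its skeletal mesh a simple physics shape.
# A single body on the root bone holds one rotated capsule, whose radius far exceeds its length, so it is nearly a large sphere.
# It has no constraints, so there is no articulation between bodies.
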PhysicsAsset /BRCosmetics/Animation/Props/Gadgets_QuarterClaspZoom_Cap/Meshes/QuarterClaspZoom_Cap_Gadget_Physics
bodies {root [{shapeType=Capsule, length=4.9692, radius=224.2685, center=<-0.0104 0.6267 0>, rotation=<rot 0 -90 0>}]}
constraints []
